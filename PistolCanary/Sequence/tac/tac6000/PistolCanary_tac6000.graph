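# PistolCanary_tac6000 is a Level Sequence in the PistolCanary tac sequence folder that acts as a container for layered subsequences.
# It holds seven Subsequences tracks, each with one section starting at frame 0 and ending near frame 450, so all play in parallel at 30 fps.
# It has no bindings of its own, so any content comes from the subsequences, whose targets are not shown.
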LevelSequence /PistolCanary/Sequence/tac/tac6000/PistolCanary_tac6000
bindings []
display_rate {denominator=1, numerator=30}
tick_resolution {denominator=1, numerator=24000}
tracks [
 {type=MovieSceneSubTrack, display_name=Subsequences, sections [{type=MovieSceneSubSection, end_frame=450, start_frame=0}]}
 {type=MovieSceneSubTrack, display_name=Subsequences, sections [{type=MovieSceneSubSection, end_frame=453, start_frame=0}]}
 {type=MovieSceneSubTrack, display_name=Subsequences, sections [{type=MovieSceneSubSection, end_frame=449, start_frame=0}]}
 {type=MovieSceneSubTrack, display_name=Subsequences, sections [{type=MovieSceneSubSection, end_frame=451, start_frame=0}]}
 {type=MovieSceneSubTrack, display_name=Subsequences, sections [{type=MovieSceneSubSection, end_frame=451, start_frame=0}]}
 {type=MovieSceneSubTrack, display_name=Subsequences, sections [{type=MovieSceneSubSection, end_frame=446, start_frame=0}]}
 {type=MovieSceneSubTrack, display_name=Subsequences, sections [{type=MovieSceneSubSection, end_frame=451, start_frame=0}]}
]
