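# Level Sequence for the animation layer of shot tac6000 in PistolCanary, animating the TentacleBoss at 30 fps.
# It possesses two boss instances, labelled Hermes_Terrain and Test Level, each with six skeletal animation sections that mostly overlap in a chain.
# The two timelines match except for the fifth section's end, and only the Test Level instance has a visibility track.
# There are no sequence-level tracks such as camera cuts or audio.
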LevelSequence /PistolCanary/Sequence/tac/tac6000/PistolCanary_tac6000_anm
bindings [
 TentacleBoss - Hermes_Terrain {
  id=#1
  possessed_class=PistolCanary_TentacleBoss_C
  tracks [{
    type=MovieSceneSkeletalAnimationTrack
    sections [
     {type=MovieSceneSkeletalAnimationSection, end_frame=130, start_frame=0}
     {type=MovieSceneSkeletalAnimationSection, end_frame=190, start_frame=116}
     {type=MovieSceneSkeletalAnimationSection, end_frame=284, start_frame=154}
     {type=MovieSceneSkeletalAnimationSection, end_frame=346, start_frame=230}
     {type=MovieSceneSkeletalAnimationSection, end_frame=422, start_frame=346}
     {type=MovieSceneSkeletalAnimationSection, end_frame=451, start_frame=397}
    ]
   }]
 }
 TentacleBoss - Test Level {
  id=#2
  possessed_class=PistolCanary_TentacleBoss_C
  tracks [
   {
    type=MovieSceneSkeletalAnimationTrack
    sections [
     {type=MovieSceneSkeletalAnimationSection, end_frame=130, start_frame=0}
     {type=MovieSceneSkeletalAnimationSection, end_frame=190, start_frame=116}
     {type=MovieSceneSkeletalAnimationSection, end_frame=284, start_frame=154}
     {type=MovieSceneSkeletalAnimationSection, end_frame=346, start_frame=230}
     {type=MovieSceneSkeletalAnimationSection, end_frame=444, start_frame=346}
     {type=MovieSceneSkeletalAnimationSection, end_frame=451, start_frame=397}
    ]
   }
   {
    type=MovieSceneVisibilityTrack
    sections [{type=MovieSceneVisibilitySection, channels [None {type=MovieSceneScriptingBoolChannel, default=true, const [true,1]}]}]
   }
  ]
 }
]
display_rate {denominator=1, numerator=30}
tick_resolution {denominator=1, numerator=24000}
tracks []
