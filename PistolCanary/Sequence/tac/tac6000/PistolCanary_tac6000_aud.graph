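# Audio Level Sequence for the tac6000 sequence in PistolCanary, scoring a tentacle boss moment.
# Several master audio tracks layer staggered, overlapping clips, including tentacle audio and a 5.1 shockwave blast that has an attach-actor binding.
# Event tracks on the possessed TentacleBoss and AudioManager actors trigger events, with the AudioManager firing twice, the second just before the shockwave clip starts.
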LevelSequence /PistolCanary/Sequence/tac/tac6000/PistolCanary_tac6000_aud
bindings [
 TentacleBoss {
  id=#1
  possessed_class=PistolCanary_TentacleBoss_C
  tracks [{
    type=MovieSceneEventTrack
    sections [{
      type=MovieSceneEventTriggerSection
      channels [None {type=MovieSceneScriptingEventChannel, const ["<Struct 'MovieSceneEvent' (0x852ff5b2bc0) {}>", 1]}]
     }]
   }]
 }
 PistolCanary_AudioManager {
  id=#2
  possessed_class=PistolCanary_AudioManager_C
  tracks [{
    type=MovieSceneEventTrack
    sections [{
      type=MovieSceneEventTriggerSection
      channels [None {
        type=MovieSceneScriptingEventChannel
        keys [[49, "<Struct 'MovieSceneEvent' (0x852ff5b2300) {}>"], [131, "<Struct 'MovieSceneEvent' (0x852ff5b2260) {}>"]]
       }]
     }]
   }]
 }
]
display_rate {denominator=1, numerator=30}
tick_resolution {denominator=1, numerator=24000}
tracks [
 {
  type=MovieSceneAudioTrack
  display_name=TentacleAudio
  sections [
   {
    type=MovieSceneAudioSection
    end_frame=243
    start_frame=-1
    channels [
     Volume {type=MovieSceneScriptingFloatChannel, default=1}
     Pitch {type=MovieSceneScriptingFloatChannel, default=1}
     AttachActor {type=MovieSceneScriptingActorReferenceChannel}
    ]
   }
   {
    type=MovieSceneAudioSection
    end_frame=404
    start_frame=159
    channels [
     Volume {type=MovieSceneScriptingFloatChannel, default=1}
     Pitch {type=MovieSceneScriptingFloatChannel, default=1}
     AttachActor {type=MovieSceneScriptingActorReferenceChannel}
    ]
   }
  ]
 }
 {
  type=MovieSceneAudioTrack
  display_name=Tentacle Shockwave Blast 5.1
  sections [{
    type=MovieSceneAudioSection
    end_frame=455
    start_frame=133
    channels [
     Volume {type=MovieSceneScriptingFloatChannel, default=1}
     Pitch {type=MovieSceneScriptingFloatChannel, default=1}
     AttachActor {type=MovieSceneScriptingActorReferenceChannel, default="<Struct 'MovieSceneObjectBindingID' (0x852ff52aa60) {}>"}
    ]
   }]
 }
 {type=MovieSceneAudioTrack, display_name=Audio, sections []}
 {
  type=MovieSceneAudioTrack
  display_name=Audio
  sections [{
    type=MovieSceneAudioSection
    end_frame=553
    start_frame=346
    channels [
     Volume {type=MovieSceneScriptingFloatChannel, default=1}
     Pitch {type=MovieSceneScriptingFloatChannel, default=1}
     AttachActor {type=MovieSceneScriptingActorReferenceChannel}
    ]
   }]
 }
 {
  type=MovieSceneAudioTrack
  display_name=Audio
  sections [
   {
    type=MovieSceneAudioSection
    end_frame=125
    start_frame=-10
    channels [
     Volume {type=MovieSceneScriptingFloatChannel, default=1}
     Pitch {type=MovieSceneScriptingFloatChannel, default=1}
     AttachActor {type=MovieSceneScriptingActorReferenceChannel}
    ]
   }
   {
    type=MovieSceneAudioSection
    end_frame=376
    start_frame=261
    channels [
     Volume {type=MovieSceneScriptingFloatChannel, default=1}
     Pitch {type=MovieSceneScriptingFloatChannel, default=1}
     AttachActor {type=MovieSceneScriptingActorReferenceChannel}
    ]
   }
  ]
 }
 {
  type=MovieSceneAudioTrack
  display_name=Audio
  sections [
   {
    type=MovieSceneAudioSection
    end_frame=285
    start_frame=136
    channels [
     Volume {type=MovieSceneScriptingFloatChannel, default=1}
     Pitch {type=MovieSceneScriptingFloatChannel, default=1}
     AttachActor {type=MovieSceneScriptingActorReferenceChannel}
    ]
   }
   {
    type=MovieSceneAudioSection
    end_frame=502
    start_frame=295
    channels [
     Volume {type=MovieSceneScriptingFloatChannel, default=1}
     Pitch {type=MovieSceneScriptingFloatChannel, default=1}
     AttachActor {type=MovieSceneScriptingActorReferenceChannel}
    ]
   }
  ]
 }
]
