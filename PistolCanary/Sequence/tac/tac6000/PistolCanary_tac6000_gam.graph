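# Gameplay-event Level Sequence for the PistolCanary tac6000 sequence, firing events on the tentacle boss.
# It binds two BGA_Tentacle_C possessables, TentacleBoss and a test-level counterpart, each with its own event trigger track.
# Both tracks fire three events at matching times, though the file does not show which functions they call.
# It has no other tracks, so it carries only event timing.
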LevelSequence /PistolCanary/Sequence/tac/tac6000/PistolCanary_tac6000_gam
bindings [
 TentacleBoss - Test Level {
  id=#1
  possessed_class=BGA_Tentacle_C
  tracks [{
    type=MovieSceneEventTrack
    sections [{
      type=MovieSceneEventTriggerSection
      channels [None {
        type=MovieSceneScriptingEventChannel
        keys [
         [132, "<Struct 'MovieSceneEvent' (0x852ff5b2a80) {}>"]
         [297, "<Struct 'MovieSceneEvent' (0x852ff5b2580) {}>"]
         [354, "<Struct 'MovieSceneEvent' (0x852ff5b24e0) {}>"]
        ]
       }]
     }]
   }]
 }
 TentacleBoss {
  id=#2
  possessed_class=BGA_Tentacle_C
  tracks [{
    type=MovieSceneEventTrack
    sections [{
      type=MovieSceneEventTriggerSection
      channels [None {
        type=MovieSceneScriptingEventChannel
        keys [
         [132, "<Struct 'MovieSceneEvent' (0x852ff5b2760) {}>"]
         [297, "<Struct 'MovieSceneEvent' (0x852ff5b2800) {}>"]
         [354, "<Struct 'MovieSceneEvent' (0x852ff5b29e0) {}>"]
        ]
       }]
     }]
   }]
 }
]
display_rate {denominator=1, numerator=30}
tick_resolution {denominator=1, numerator=24000}
tracks []
