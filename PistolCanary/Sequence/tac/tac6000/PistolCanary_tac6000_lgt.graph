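# Lighting Level Sequence for shot tac6000 that animates three point lights named tentacleCharge1 to tentacleCharge3.
# Spawn tracks show each light in its own successive, non-overlapping window, while transform tracks move it through the scene.
# The first two follow nearly the same path, a slow climb, brief hold and fast drop; the third is a much shorter rise and fall.
# Intensity and white color stay constant, so only position and spawn timing change.
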LevelSequence /PistolCanary/Sequence/tac/tac6000/PistolCanary_tac6000_lgt
bindings [
 tentacleCharge1 {
  id=#1
  possessed_class=PointLight
  tracks [
   {
    type=MovieSceneSpawnTrack
    sections [{type=MovieSceneSpawnSection, channels [None {type=MovieSceneScriptingBoolChannel, default=false, keys [[11,false],[17,true],[133,false]]}]}]
   }
   {
    type=MovieScene3DTransformTrack
    sections [{
      type=MovieScene3DTransformSection
      channels [
       Location.X {
        type=MovieSceneScriptingDoubleChannel
        default=14358.2387
        keys [[17,13644.9756],[47,13529.0898],[84,15750.8945],[120,16498.1465],[125,16498.1465],[128,15533.25],[130,14251.0068],[132,14358.2383]]
       }
       Location.Y {
        type=MovieSceneScriptingDoubleChannel
        default=70522.7359
        keys [[17,71164.2188],[47,70839.0469],[84,68237.4453],[120,66771.7969],[125,66771.7969],[128,68163.5312],[130,70027.7266],[132,70522.7344]]
       }
       Location.Z {
        type=MovieSceneScriptingDoubleChannel
        default=4443.5472
        keys [[17,3045.6792],[120,11908.5664],[125,11908.5664],[128,9718.5508],[132,3335.6086]]
       }
       Rotation.X {type=MovieSceneScriptingDoubleChannel, default=0}
       Rotation.Y {type=MovieSceneScriptingDoubleChannel, default=90}
       Rotation.Z {type=MovieSceneScriptingDoubleChannel, default=0}
       Scale.X {type=MovieSceneScriptingDoubleChannel, default=1}
       Scale.Y {type=MovieSceneScriptingDoubleChannel, default=1}
       Scale.Z {type=MovieSceneScriptingDoubleChannel, default=1}
      ]
     }]
   }
  ]
 }
 tentacleCharge3 {
  id=#2
  possessed_class=PointLight
  tracks [
   {
    type=MovieSceneSpawnTrack
    sections [{type=MovieSceneSpawnSection, channels [None {type=MovieSceneScriptingBoolChannel, default=false, keys [[339,false],[342,true],[354,false]]}]}]
   }
   {
    type=MovieScene3DTransformTrack
    sections [{
      type=MovieScene3DTransformSection
      channels [
       Location.X {type=MovieSceneScriptingDoubleChannel, default=14358.2387, keys [[343,14059.0195],[346,13684.4365],[348,14251.0068],[353,14358.2383]]}
       Location.Y {type=MovieSceneScriptingDoubleChannel, default=70522.7359, keys [[343,68663.1484],[346,68905.7812],[348,70027.7266],[353,70522.7344]]}
       Location.Z {type=MovieSceneScriptingDoubleChannel, default=4443.5472, keys [[342,1769.151],[343,10160.3096],[346,9407.0576],[353,3335.6086]]}
       Rotation.X {type=MovieSceneScriptingDoubleChannel, default=0}
       Rotation.Y {type=MovieSceneScriptingDoubleChannel, default=90}
       Rotation.Z {type=MovieSceneScriptingDoubleChannel, default=0}
       Scale.X {type=MovieSceneScriptingDoubleChannel, default=1}
       Scale.Y {type=MovieSceneScriptingDoubleChannel, default=1}
       Scale.Z {type=MovieSceneScriptingDoubleChannel, default=1}
      ]
     }]
   }
  ]
 }
 LightComponent0 {
  id=#3
  parent_id=#4
  possessed_class=PointLightComponent
  tracks [
   {
    type=MovieSceneFloatTrack
    display_name=Intensity
    sections [{type=MovieSceneFloatSection, channels [None {type=MovieSceneScriptingFloatChannel, default=800}]}]
   }
   {
    type=MovieSceneColorTrack
    display_name=Light Color
    sections [{
      type=MovieSceneColorSection
      channels [
       Color.R {type=MovieSceneScriptingFloatChannel, default=1}
       Color.G {type=MovieSceneScriptingFloatChannel, default=1}
       Color.B {type=MovieSceneScriptingFloatChannel, default=1}
       Color.A {type=MovieSceneScriptingFloatChannel, default=1}
      ]
     }]
   }
  ]
 }
 LightComponent0 {
  id=#5
  parent_id=#1
  possessed_class=PointLightComponent
  tracks [
   {
    type=MovieSceneFloatTrack
    display_name=Intensity
    sections [{type=MovieSceneFloatSection, channels [None {type=MovieSceneScriptingFloatChannel, default=800}]}]
   }
   {
    type=MovieSceneColorTrack
    display_name=Light Color
    sections [{
      type=MovieSceneColorSection
      channels [
       Color.R {type=MovieSceneScriptingFloatChannel, default=1}
       Color.G {type=MovieSceneScriptingFloatChannel, default=1}
       Color.B {type=MovieSceneScriptingFloatChannel, default=1}
       Color.A {type=MovieSceneScriptingFloatChannel, default=1}
      ]
     }]
   }
  ]
 }
 LightComponent0 {
  id=#6
  parent_id=#2
  possessed_class=PointLightComponent
  tracks [
   {
    type=MovieSceneFloatTrack
    display_name=Intensity
    sections [{type=MovieSceneFloatSection, channels [None {type=MovieSceneScriptingFloatChannel, default=800}]}]
   }
   {
    type=MovieSceneColorTrack
    display_name=Light Color
    sections [{
      type=MovieSceneColorSection
      channels [
       Color.R {type=MovieSceneScriptingFloatChannel, default=1}
       Color.G {type=MovieSceneScriptingFloatChannel, default=1}
       Color.B {type=MovieSceneScriptingFloatChannel, default=1}
       Color.A {type=MovieSceneScriptingFloatChannel, default=1}
      ]
     }]
   }
  ]
 }
 tentacleCharge2 {
  id=#4
  possessed_class=PointLight
  tracks [
   {
    type=MovieSceneSpawnTrack
    sections [{
      type=MovieSceneSpawnSection
      channels [None {type=MovieSceneScriptingBoolChannel, default=false, keys [[143,false],[182.9975,true],[304,false]]}]
     }]
   }
   {
    type=MovieScene3DTransformTrack
    sections [{
      type=MovieScene3DTransformSection
      channels [
       Location.X {
        type=MovieSceneScriptingDoubleChannel
        default=14358.2387
        keys [[182.9975,14358.2383],[212.9975,13529.0898],[249.9975,15750.8945],[285.9975,16498.1465],[290.9975,16498.1465],[293.9975,15533.25],[295.9975,14251.0068],[303,14358.2383]]
       }
       Location.Y {
        type=MovieSceneScriptingDoubleChannel
        default=70522.7359
        keys [[182.9975,70522.7344],[212.9975,70839.0469],[249.9975,68237.4453],[285.9975,66771.7969],[290.9975,66771.7969],[293.9975,68163.5312],[295.9975,70027.7266],[303,70522.7344]]
       }
       Location.Z {
        type=MovieSceneScriptingDoubleChannel
        default=4443.5472
        keys [[182.9975,3586.9451],[285.9975,11908.5664],[290.9975,11908.5664],[293.9975,9718.5508],[303,3335.6086]]
       }
       Rotation.X {type=MovieSceneScriptingDoubleChannel, default=0}
       Rotation.Y {type=MovieSceneScriptingDoubleChannel, default=90}
       Rotation.Z {type=MovieSceneScriptingDoubleChannel, default=0}
       Scale.X {type=MovieSceneScriptingDoubleChannel, default=1}
       Scale.Y {type=MovieSceneScriptingDoubleChannel, default=1}
       Scale.Z {type=MovieSceneScriptingDoubleChannel, default=1}
      ]
     }]
   }
  ]
 }
]
display_rate {denominator=1, numerator=30}
tick_resolution {denominator=1, numerator=24000}
tracks []
