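# VFX Level Sequence for the tac6000 tentacle boss attack, layering charge-up, body glow and wave effects.
# Material tracks animate pulse offset and emissive color on four TentacleBoss body slots, alongside two event triggers on the actor.
# Two attached charge-up laser Niagara systems use animated parameters, while three TripleTentacleWave systems activate and deactivate in staggered windows.
# Four temp wave meshes scale up and sink, with visibility keyed off.
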LevelSequence /PistolCanary/Sequence/tac/tac6000/PistolCanary_tac6000_vfx
bindings [
 tempWaveo2 {
  id=#1
  possessed_class=FortStaticMeshActor
  tracks [
   {type=MovieSceneSpawnTrack, sections [{type=MovieSceneSpawnSection, channels [None {type=MovieSceneScriptingBoolChannel, default=true}]}]}
   {
    type=MovieScene3DTransformTrack
    sections [{
      type=MovieScene3DTransformSection
      channels [
       Location.X {type=MovieSceneScriptingDoubleChannel, default=14739.1414, const [14739.1416,3]}
       Location.Y {type=MovieSceneScriptingDoubleChannel, default=68065.8883, const [68065.8906,3]}
       Location.Z {type=MovieSceneScriptingDoubleChannel, default=-12377, keys [[286,2154.8146],[312,2154.8146],[405,-12377]]}
       Rotation.X {type=MovieSceneScriptingDoubleChannel, default=0, const [0,3]}
       Rotation.Y {type=MovieSceneScriptingDoubleChannel, default=0, const [0,3]}
       Rotation.Z {type=MovieSceneScriptingDoubleChannel, default=0, const [0,3]}
       Scale.X {type=MovieSceneScriptingDoubleChannel, default=68, thinned_from=3, keys [[286,0.1],[354,1009.377]]}
       Scale.Y {type=MovieSceneScriptingDoubleChannel, default=68, thinned_from=3, keys [[286,0.1],[354,1009.377]]}
       Scale.Z {type=MovieSceneScriptingDoubleChannel, default=50, thinned_from=3, keys [[286,0.1],[354,130.6077]]}
      ]
     }]
   }
   {
    type=MovieSceneVisibilityTrack
    sections [{type=MovieSceneVisibilitySection, channels [None {type=MovieSceneScriptingBoolChannel, default=false}]}]
   }
  ]
 }
 SkeletalMesh {
  id=#2
  parent_id=#3
  possessed_class=SkeletalMeshComponent
  tracks [
   {
    type=MovieSceneComponentMaterialTrack
    display_name=Material Slot: Body_A
    sections [{
      type=MovieSceneComponentMaterialParameterSection
      end_frame=444
      start_frame=30
      channels [
       PulseSpeed {type=MovieSceneScriptingFloatChannel, default=0, const [0,1]}
       PulseManualOffset {
        type=MovieSceneScriptingFloatChannel
        default=0
        keys [[43.5263,0],[97.6313,1],[97.9125,0],[137.2088,0],[141.7175,1],[142.2188,0],[300,0],[304.5087,1],[305.01,0],[340.58,0],[345.0888,1],[345.5888,0]]
       }
       GlowPulseEXP {type=MovieSceneScriptingFloatChannel, default=0.079, keys [[32.0038,0.079],[39.0175,0.069],[44.0275,0.079]]}
       EmissiveGlowColorR {type=MovieSceneScriptingFloatChannel, const [12,2]}
       EmissiveGlowColorG {type=MovieSceneScriptingFloatChannel, const [2.2855,2]}
       EmissiveGlowColorB {type=MovieSceneScriptingFloatChannel, default=974.9821, keys [[43.5263,1.8839],[67.0725,100],[346,100],[383.5725,1.8839]]}
       EmissiveGlowColorA {type=MovieSceneScriptingFloatChannel, const [1,2]}
      ]
     }]
   }
   {
    type=MovieSceneComponentMaterialTrack
    display_name=Material Slot: Body_B
    sections [{
      type=MovieSceneComponentMaterialParameterSection
      end_frame=443
      start_frame=32
      channels [
       PulseManualOffset {
        type=MovieSceneScriptingFloatChannel
        default=0
        keys [[43.5263,0],[97.6313,1],[97.9125,0],[137.2088,0],[141.7175,1],[142.2188,0],[300,0],[304.5087,1],[305.01,0],[340.58,0],[345.0888,1],[345.5888,0]]
       }
       PulseSpeed {type=MovieSceneScriptingFloatChannel, default=0, const [0,1]}
       GlowPulseEXP {type=MovieSceneScriptingFloatChannel, default=0.079, keys [[32.0038,0.079],[39.0175,0.069],[44.0275,0.079]]}
       EmissiveGlowColorR {type=MovieSceneScriptingFloatChannel, const [12,2]}
       EmissiveGlowColorG {type=MovieSceneScriptingFloatChannel, const [2.2855,2]}
       EmissiveGlowColorB {type=MovieSceneScriptingFloatChannel, default=1.8839, keys [[46.5325,1.8839],[67.5725,100],[346,100],[383.5725,1.8839]]}
       EmissiveGlowColorA {type=MovieSceneScriptingFloatChannel, const [1,2]}
      ]
     }]
   }
   {
    type=MovieSceneComponentMaterialTrack
    display_name=Material Slot: Body_C
    sections [{
      type=MovieSceneComponentMaterialParameterSection
      end_frame=444
      start_frame=30
      channels [
       PulseSpeed {type=MovieSceneScriptingFloatChannel, default=0, const [0,1]}
       PulseManualOffset {
        type=MovieSceneScriptingFloatChannel
        default=0
        keys [[43.5263,0],[97.6313,1],[97.9125,0],[137.2088,0],[141.7175,1],[142.2188,0],[300,0],[304.5087,1],[305.01,0],[340.58,0],[345.0888,1],[345.5888,0]]
       }
       GlowPulseEXP {type=MovieSceneScriptingFloatChannel, keys [[31.5025,0.079],[38.5163,0.069],[43.5263,0.079],[154.2412,0.079]]}
       EmissiveGlowColorR {type=MovieSceneScriptingFloatChannel, const [12,2]}
       EmissiveGlowColorG {type=MovieSceneScriptingFloatChannel, const [2.2855,2]}
       EmissiveGlowColorB {type=MovieSceneScriptingFloatChannel, default=1.8839, keys [[49.5375,1.8839],[68.0738,100],[346,100],[383.5725,1.8839]]}
       EmissiveGlowColorA {type=MovieSceneScriptingFloatChannel, const [1,2]}
      ]
     }]
   }
   {
    type=MovieSceneComponentMaterialTrack
    display_name=Material Slot: Body_D
    sections [{
      type=MovieSceneComponentMaterialParameterSection
      end_frame=444
      start_frame=31
      channels [
       PulseManualOffset {
        type=MovieSceneScriptingFloatChannel
        default=0
        keys [[43.5263,0],[97.6313,1],[97.9125,0],[137.2088,0],[141.7175,1],[142.2188,0],[300,0],[304.5087,1],[305.01,0],[340.58,0],[345.0888,1],[345.5888,0]]
       }
       PulseSpeed {type=MovieSceneScriptingFloatChannel, default=0, const [0,2]}
       GlowPulseEXP {type=MovieSceneScriptingFloatChannel, keys [[31.5025,0.079],[38.5163,0.069],[43.5263,0.079]]}
       EmissiveGlowColorR {type=MovieSceneScriptingFloatChannel, const [12,2]}
       EmissiveGlowColorG {type=MovieSceneScriptingFloatChannel, const [2.2855,2]}
       EmissiveGlowColorB {type=MovieSceneScriptingFloatChannel, default=1.8839, keys [[52.0425,1.8839],[68.0738,100],[346,100],[383.5725,1.8839]]}
       EmissiveGlowColorA {type=MovieSceneScriptingFloatChannel, const [1,2]}
      ]
     }]
   }
  ]
 }
 TentacleBoss {
  id=#3
  possessed_class=BGA_Tentacle_C
  tracks [{
    type=MovieSceneEventTrack
    sections [{
      type=MovieSceneEventTriggerSection
      channels [None {
        type=MovieSceneScriptingEventChannel
        keys [[50, "<Struct 'MovieSceneEvent' (0x852ff5b4d80) {}>"], [340, "<Struct 'MovieSceneEvent' (0x852ff5b4ce0) {}>"]]
       }]
     }]
   }]
 }
 NS_PistolCanary_TripleTentacleWave2 {
  id=#4
  possessed_class=NiagaraActor
  tracks [{
    type=MovieScene3DTransformTrack
    sections [{
      type=MovieScene3DTransformSection
      channels [
       Location.X {type=MovieSceneScriptingDoubleChannel, default=14388, const [14600,1]}
       Location.Y {type=MovieSceneScriptingDoubleChannel, default=70184, const [69752,1]}
       Location.Z {type=MovieSceneScriptingDoubleChannel, default=4648, const [3980,1]}
       Rotation.X {type=MovieSceneScriptingDoubleChannel, default=0, const [0,1]}
       Rotation.Y {type=MovieSceneScriptingDoubleChannel, default=0, const [0,1]}
       Rotation.Z {type=MovieSceneScriptingDoubleChannel, default=0, const [0,1]}
       Scale.X {type=MovieSceneScriptingDoubleChannel, default=1, const [1,1]}
       Scale.Y {type=MovieSceneScriptingDoubleChannel, default=1, const [1,1]}
       Scale.Z {type=MovieSceneScriptingDoubleChannel, default=1, const [1,1]}
      ]
     }]
   }]
 }
 tempWaveo4 {
  id=#5
  possessed_class=FortStaticMeshActor
  tracks [
   {type=MovieSceneSpawnTrack, sections [{type=MovieSceneSpawnSection, channels [None {type=MovieSceneScriptingBoolChannel, default=true}]}]}
   {
    type=MovieScene3DTransformTrack
    sections [{
      type=MovieScene3DTransformSection
      channels [
       Location.X {type=MovieSceneScriptingDoubleChannel, default=14739.1414, const [14739.1416,3]}
       Location.Y {type=MovieSceneScriptingDoubleChannel, default=67640, keys [[136,68065.8906],[162,67640],[204,68065.8906]]}
       Location.Z {type=MovieSceneScriptingDoubleChannel, default=2295, keys [[136,3123.2514],[153,2758],[158,2295],[162,1916],[175,449],[255,-12377]]}
       Rotation.X {type=MovieSceneScriptingDoubleChannel, default=0, const [0,3]}
       Rotation.Y {type=MovieSceneScriptingDoubleChannel, default=0, const [0,3]}
       Rotation.Z {type=MovieSceneScriptingDoubleChannel, default=0, const [0,3]}
       Scale.X {type=MovieSceneScriptingDoubleChannel, default=68, thinned_from=3, keys [[136,0.1],[204,1009.377]]}
       Scale.Y {type=MovieSceneScriptingDoubleChannel, default=68, thinned_from=3, keys [[136,0.1],[204,1009.377]]}
       Scale.Z {type=MovieSceneScriptingDoubleChannel, default=50, thinned_from=3, keys [[136,0.1],[204,130.6077]]}
      ]
     }]
   }
   {
    type=MovieSceneVisibilityTrack
    sections [{type=MovieSceneVisibilitySection, channels [None {type=MovieSceneScriptingBoolChannel, default=false}]}]
   }
  ]
 }
 NiagaraComponent0 {
  id=#6
  parent_id=#4
  possessed_class=NiagaraComponent
  tracks [
   {
    type=MovieSceneParticleTrack
    display_name=FX System
    sections [{
      type=MovieSceneParticleSection
      channels [None {type=MovieSceneScriptingParticleChannel, keys [[142, "<ParticleKey.ACTIVATE: 0>"], [213, "<ParticleKey.DEACTIVATE: 1>"]]}]
     }]
   }
   {type=MovieSceneNiagaraSystemTrack, sections [{type=MovieSceneNiagaraSystemSpawnSection, end_frame=453, start_frame=142}]}
  ]
 }
 NiagaraComponent0 {
  id=#7
  parent_id=#8
  possessed_class=NiagaraComponent
  tracks [
   {type=MovieSceneNiagaraSystemTrack, sections [{type=MovieSceneNiagaraSystemSpawnSection, end_frame=502, start_frame=202}]}
   {
    type=MovieSceneNiagaraFloatParameterTrack
    display_name=ChargeProgress
    sections [{type=MovieSceneFloatSection, channels [None {type=MovieSceneScriptingFloatChannel, default=1, const [0,3]}]}]
   }
   {
    type=MovieSceneNiagaraFloatParameterTrack
    display_name=PlasmaEnabled
    sections [{type=MovieSceneFloatSection, channels [None {type=MovieSceneScriptingFloatChannel, default=0, const [0,2]}]}]
   }
   {
    type=MovieSceneNiagaraFloatParameterTrack
    display_name=ChargeProgressLightning
    sections [{type=MovieSceneFloatSection, channels [None {type=MovieSceneScriptingFloatChannel, default=0.5472, const [0,3]}]}]
   }
   {
    type=MovieSceneNiagaraFloatParameterTrack
    display_name=PostProcess
    sections [{
      type=MovieSceneFloatSection
      channels [None {type=MovieSceneScriptingFloatChannel, default=0, thinned_from=4, keys [[203.5,0],[293.5,0],[382,0.5]]}]
     }]
   }
   {
    type=MovieSceneNiagaraFloatParameterTrack
    display_name=TipDelay
    sections [{type=MovieSceneFloatSection, channels [None {type=MovieSceneScriptingFloatChannel, default=1}]}]
   }
   {
    type=MovieSceneParticleTrack
    display_name=FX System
    sections [{type=MovieSceneParticleSection, channels [None {type=MovieSceneScriptingParticleChannel, const ["<ParticleKey.ACTIVATE: 0>", 1]}]}]
   }
   {
    type=MovieSceneNiagaraColorParameterTrack
    display_name=ChargeScaleColor
    sections [{
      type=MovieSceneColorSection
      channels [
       Color.R {type=MovieSceneScriptingFloatChannel, default=1}
       Color.G {type=MovieSceneScriptingFloatChannel, default=1}
       Color.B {type=MovieSceneScriptingFloatChannel, default=1}
       Color.A {type=MovieSceneScriptingFloatChannel, default=0, keys [[204,0],[221.5,1],[360,1],[410,0]]}
      ]
     }]
   }
   {
    type=MovieSceneNiagaraVectorParameterTrack
    display_name=ChargeScaleVec
    sections [{
      type=MovieSceneFloatVectorSection
      channels [
       Vector.X {type=MovieSceneScriptingFloatChannel, default=0.6, keys [[222,0],[224.5,0.6],[269.5,1]]}
       Vector.Y {type=MovieSceneScriptingFloatChannel, default=1, const [1,2]}
       Vector.Z {type=MovieSceneScriptingFloatChannel, default=1, const [1,2]}
      ]
     }]
   }
  ]
 }
 tempWaveo01 {
  id=#9
  possessed_class=FortStaticMeshActor
  tracks [
   {type=MovieSceneSpawnTrack, sections [{type=MovieSceneSpawnSection, channels [None {type=MovieSceneScriptingBoolChannel, default=true}]}]}
   {
    type=MovieScene3DTransformTrack
    sections [{
      type=MovieScene3DTransformSection
      channels [
       Location.X {type=MovieSceneScriptingDoubleChannel, default=14739.1414, const [14739.1416,3]}
       Location.Y {type=MovieSceneScriptingDoubleChannel, default=68065.8883, const [68065.8906,3]}
       Location.Z {type=MovieSceneScriptingDoubleChannel, default=2043, keys [[136,2154.8146],[153,2043],[162,-2768],[175,-4494],[255,-12377]]}
       Rotation.X {type=MovieSceneScriptingDoubleChannel, default=0, const [0,3]}
       Rotation.Y {type=MovieSceneScriptingDoubleChannel, default=0, const [0,3]}
       Rotation.Z {type=MovieSceneScriptingDoubleChannel, default=0, const [0,3]}
       Scale.X {type=MovieSceneScriptingDoubleChannel, default=68, thinned_from=3, keys [[136,0.1],[204,1009.377]]}
       Scale.Y {type=MovieSceneScriptingDoubleChannel, default=68, thinned_from=3, keys [[136,0.1],[204,1009.377]]}
       Scale.Z {type=MovieSceneScriptingDoubleChannel, default=50, thinned_from=3, keys [[136,0.1],[204,130.6077]]}
      ]
     }]
   }
   {
    type=MovieSceneVisibilityTrack
    sections [{type=MovieSceneVisibilitySection, channels [None {type=MovieSceneScriptingBoolChannel, default=false}]}]
   }
  ]
 }
 NS_PistolCanary_TripleTentacleWave3 {
  id=#10
  possessed_class=NiagaraActor
  tracks [{
    type=MovieScene3DTransformTrack
    sections [{
      type=MovieScene3DTransformSection
      channels [
       Location.X {type=MovieSceneScriptingDoubleChannel, default=14388, keys [[153,14336],[294,14508]]}
       Location.Y {type=MovieSceneScriptingDoubleChannel, default=70184, keys [[153,67604],[294,69484]]}
       Location.Z {type=MovieSceneScriptingDoubleChannel, default=4648, const [3912,2]}
       Rotation.X {type=MovieSceneScriptingDoubleChannel, default=0, const [0,2]}
       Rotation.Y {type=MovieSceneScriptingDoubleChannel, default=0, const [0,2]}
       Rotation.Z {type=MovieSceneScriptingDoubleChannel, default=0, const [0,2]}
       Scale.X {type=MovieSceneScriptingDoubleChannel, default=1, const [1,2]}
       Scale.Y {type=MovieSceneScriptingDoubleChannel, default=1, const [1,2]}
       Scale.Z {type=MovieSceneScriptingDoubleChannel, default=1, const [1,2]}
      ]
     }]
   }]
 }
 NiagaraComponent0 {
  id=#11
  parent_id=#12
  possessed_class=NiagaraComponent
  tracks [
   {
    type=MovieSceneParticleTrack
    display_name=FX System
    sections [{
      type=MovieSceneParticleSection
      channels [None {type=MovieSceneScriptingParticleChannel, keys [[355, "<ParticleKey.ACTIVATE: 0>"], [433, "<ParticleKey.DEACTIVATE: 1>"]]}]
     }]
   }
   {type=MovieSceneNiagaraSystemTrack, sections [{type=MovieSceneNiagaraSystemSpawnSection, end_frame=442, start_frame=355}]}
  ]
 }
 NS_PistolCanary_TripleTentacleWave4 {
  id=#12
  possessed_class=NiagaraActor
  tracks [
   {
    type=MovieScene3DTransformTrack
    sections [{
      type=MovieScene3DTransformSection
      channels [
       Location.X {type=MovieSceneScriptingDoubleChannel, default=14388, const [14336,2]}
       Location.Y {type=MovieSceneScriptingDoubleChannel, default=70184, keys [[153,67604],[350,68988]]}
       Location.Z {type=MovieSceneScriptingDoubleChannel, default=4648, const [3856,2]}
       Rotation.X {type=MovieSceneScriptingDoubleChannel, default=0, const [0,2]}
       Rotation.Y {type=MovieSceneScriptingDoubleChannel, default=0, const [0,2]}
       Rotation.Z {type=MovieSceneScriptingDoubleChannel, default=0, const [0,2]}
       Scale.X {type=MovieSceneScriptingDoubleChannel, default=1, const [1,2]}
       Scale.Y {type=MovieSceneScriptingDoubleChannel, default=1, const [1,2]}
       Scale.Z {type=MovieSceneScriptingDoubleChannel, default=1, const [1,2]}
      ]
     }]
   }
   {
    type=MovieSceneVisibilityTrack
    sections [{type=MovieSceneVisibilitySection, channels [None {type=MovieSceneScriptingBoolChannel, default=true}]}]
   }
  ]
 }
 tempWaveo3 {
  id=#13
  possessed_class=FortStaticMeshActor
  tracks [
   {type=MovieSceneSpawnTrack, sections [{type=MovieSceneSpawnSection, channels [None {type=MovieSceneScriptingBoolChannel, default=true}]}]}
   {
    type=MovieScene3DTransformTrack
    sections [{
      type=MovieScene3DTransformSection
      channels [
       Location.X {type=MovieSceneScriptingDoubleChannel, default=14739.1414, const [14739.1416,3]}
       Location.Y {type=MovieSceneScriptingDoubleChannel, default=68065.8883, const [68065.8906,3]}
       Location.Z {type=MovieSceneScriptingDoubleChannel, default=-12377, keys [[339,2154.8146],[365,2154.8146],[458,-12377]]}
       Rotation.X {type=MovieSceneScriptingDoubleChannel, default=0, const [0,3]}
       Rotation.Y {type=MovieSceneScriptingDoubleChannel, default=0, const [0,3]}
       Rotation.Z {type=MovieSceneScriptingDoubleChannel, default=0, const [0,3]}
       Scale.X {type=MovieSceneScriptingDoubleChannel, default=68, thinned_from=3, keys [[339,0.1],[407,1009.377]]}
       Scale.Y {type=MovieSceneScriptingDoubleChannel, default=68, thinned_from=3, keys [[339,0.1],[407,1009.377]]}
       Scale.Z {type=MovieSceneScriptingDoubleChannel, default=50, thinned_from=3, keys [[339,0.1],[407,130.6077]]}
      ]
     }]
   }
   {
    type=MovieSceneVisibilityTrack
    sections [{type=MovieSceneVisibilitySection, channels [None {type=MovieSceneScriptingBoolChannel, default=false}]}]
   }
  ]
 }
 NiagaraComponent0 {
  id=#14
  parent_id=#15
  possessed_class=NiagaraComponent
  tracks [
   {type=MovieSceneNiagaraSystemTrack, sections [{type=MovieSceneNiagaraSystemSpawnSection, end_frame=427, start_frame=29}]}
   {
    type=MovieSceneNiagaraVectorParameterTrack
    display_name=ChargeScaleVec
    sections [{
      type=MovieSceneFloatVectorSection
      channels [
       Vector.X {type=MovieSceneScriptingFloatChannel, default=0.6, keys [[30,0],[32.5,0.6],[77.5,1]]}
       Vector.Y {type=MovieSceneScriptingFloatChannel, default=1, const [1,2]}
       Vector.Z {type=MovieSceneScriptingFloatChannel, default=1, const [1,2]}
      ]
     }]
   }
   {
    type=MovieSceneNiagaraFloatParameterTrack
    display_name=ChargeProgress
    sections [{type=MovieSceneFloatSection, channels [None {type=MovieSceneScriptingFloatChannel, default=1, keys [[29.5,0],[45.5,0.3],[75,1]]}]}]
   }
   {
    type=MovieSceneNiagaraFloatParameterTrack
    display_name=PlasmaEnabled
    sections [{type=MovieSceneFloatSection, channels [None {type=MovieSceneScriptingFloatChannel, default=0, keys [[78,1],[82.5,0]]}]}]
   }
   {
    type=MovieSceneNiagaraColorParameterTrack
    display_name=ChargeScaleColor
    sections [{
      type=MovieSceneColorSection
      channels [
       Color.R {type=MovieSceneScriptingFloatChannel, default=1}
       Color.G {type=MovieSceneScriptingFloatChannel, default=1}
       Color.B {type=MovieSceneScriptingFloatChannel, default=1}
       Color.A {type=MovieSceneScriptingFloatChannel, default=0, keys [[37,0],[54.5,1],[236,1],[286,0]]}
      ]
     }]
   }
   {
    type=MovieSceneNiagaraFloatParameterTrack
    display_name=ChargeProgressLightning
    sections [{type=MovieSceneFloatSection, channels [None {type=MovieSceneScriptingFloatChannel, default=0.5472, keys [[30.5,0],[58.5,0.5472],[106.5,1]]}]}]
   }
   {
    type=MovieSceneNiagaraFloatParameterTrack
    display_name=PostProcess
    sections [{type=MovieSceneFloatSection, channels [None {type=MovieSceneScriptingFloatChannel, default=0.2, keys [[31,0],[85,0.2],[121,0.4]]}]}]
   }
   {
    type=MovieSceneNiagaraFloatParameterTrack
    display_name=TipDelay
    sections [{type=MovieSceneFloatSection, channels [None {type=MovieSceneScriptingFloatChannel, default=1}]}]
   }
   {
    type=MovieSceneParticleTrack
    display_name=FX System
    sections [{type=MovieSceneParticleSection, channels [None {type=MovieSceneScriptingParticleChannel, const ["<ParticleKey.ACTIVATE: 0>", 1]}]}]
   }
  ]
 }
 NiagaraComponent0 {
  id=#16
  parent_id=#10
  possessed_class=NiagaraComponent
  tracks [
   {
    type=MovieSceneParticleTrack
    display_name=FX System
    sections [{
      type=MovieSceneParticleSection
      channels [None {type=MovieSceneScriptingParticleChannel, keys [[306, "<ParticleKey.ACTIVATE: 0>"], [379, "<ParticleKey.DEACTIVATE: 1>"]]}]
     }]
   }
   {type=MovieSceneNiagaraSystemTrack, sections [{type=MovieSceneNiagaraSystemSpawnSection, end_frame=393, start_frame=306}]}
  ]
 }
 NS_TentacleAttackChargeUpLaser4 {
  id=#15
  possessed_class=NiagaraActor
  tracks [{type=MovieScene3DAttachTrack, sections [{type=MovieScene3DAttachSection, end_frame=417, start_frame=28}]}]
 }
 NS_TentacleAttackChargeUpLaser5 {
  id=#8
  possessed_class=NiagaraActor
  tracks [{type=MovieScene3DAttachTrack, sections [{type=MovieScene3DAttachSection, end_frame=504, start_frame=201}]}]
 }
]
display_rate {denominator=1, numerator=30}
tick_resolution {denominator=1, numerator=24000}
tracks []
